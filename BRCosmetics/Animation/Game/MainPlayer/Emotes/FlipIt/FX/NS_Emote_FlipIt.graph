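# Niagara system for the FlipIt emote that draws orange ribbon trails, named for the hands, from a skeletal mesh's animation.
# A single Trail_Hands emitter uses the FN anim-trail modules to spawn, initialize and update particles against User.SkeletalMesh.
# User.FXTrailScale drives trail width and lifetime, defaulting to zero.
# Curves taper width and fade alpha over particle life, with alpha scaled by time of day in gameplay and fixed in the front end.
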NiagaraSystem /BRCosmetics/Animation/Game/MainPlayer/Emotes/FlipIt/FX/NS_Emote_FlipIt
dependencies {usedDataInterfaces [&NiagaraDataInterfaceCurve], usedDynamicInputs [], usedModules [&SystemState]}
emitters {Trail_Hands {
  emitterUpdateScript [
   &EmitterState()
   &FN_SpawnAnimTrail()
  ]
  particleSpawnScript [
   &FN_InitializeAnimTrailBoneIndecies()
   &FN_InitializeAnimTrail(
    AnimTrailWidth=&Lerp_Float(A=0.25, B=1.5, Alpha=$User.FXTrailScale)
    Color=<rgba 0.75 0.2675 0 1>
    Lifetime=&FN_FrontEndFloatModifier(
     Gameplay=&Lerp_Float(
      A=0
      B=0.75
      Alpha=&FloatFromCurve(
       FloatCurve=DataInterface() {propertyValues {Curve {keys [
           {arriveTangent=-1.836, interpMode=RCIM_Cubic, leaveTangent=-1.836, tangentMode=RCTM_User, time=0, value=1}
           {interpMode=RCIM_Cubic, tangentMode=RCTM_User, time=1, value=0}
          ]}}}
       Scale Curve=$User.FXTrailScale
      )
     )
     Front End=&Lerp_Float(
      A=0
      B=0.75
      Alpha=&FloatFromCurve(
       FloatCurve=DataInterface() {propertyValues {Curve {keys [
           {arriveTangent=-1.836, interpMode=RCIM_Cubic, leaveTangent=-1.836, tangentMode=RCTM_User, time=0, value=1}
           {interpMode=RCIM_Cubic, tangentMode=RCTM_User, time=1, value=0}
          ]}}}
       Scale Curve=$User.FXTrailScale
      )
     )
    )
   )
  ]
  particleUpdateScript [
   &SolveForcesAndVelocity(Clamp Velocity=false, Limit Acceleration=false)
   &ParticleState()
   &FN_UpdateAnimTrail(SkeletalMesh=$User.SkeletalMesh)
   &ScaleRibbonWidth(Ribbon Width Scale=&FloatFromCurve(FloatCurve=DataInterface() {propertyValues {Curve {keys [
         {arriveTangent=6.4514, interpMode=RCIM_Cubic, leaveTangent=6.4514, tangentMode=RCTM_User, time=0, value=0.5}
         {arriveTangent=-0.7143, interpMode=RCIM_Cubic, leaveTangent=-0.7143, time=0.1, value=1}
         {interpMode=RCIM_Cubic, tangentMode=RCTM_User, time=0.7, value=0}
        ]}}}))
   &ScaleColor(
    Scale Alpha=&FloatFromCurve(
     FloatCurve=DataInterface() {propertyValues {Curve {keys [
         {arriveTangent=-2.4147, interpMode=RCIM_Cubic, leaveTangent=-2.4147, tangentMode=RCTM_User, time=0, value=1}
         {interpMode=RCIM_Cubic, tangentMode=RCTM_User, time=0.7, value=0}
        ]}}}
     Scale Curve=&FN_FrontEndFloatModifier(
      Gameplay=&FNTimeOfDayScalar(Night Scale=0.5, Morning Scale=0.5, Day Scale=1.1, Evening Scale=0.5)
      Front End=1.25
     )
    )
    ScaleA=true
    ScaleRGB=true
   )
  ]
  rendererClasses [&NiagaraRibbonRendererProperties]
  renderers [{rendererClass=&NiagaraRibbonRendererProperties}]
 }}
summary {
 systemName=NS_Emote_FlipIt
 emitters [{emitterName=Trail_Hands, rendererClasses [&NiagaraRibbonRendererProperties]}]
 userVariables [
  User.bIsFrontEnd {type=NiagaraBool}
  User.bIsFrontEndPreview {type=NiagaraBool}
  User.FXTrailScale {type=NiagaraFloat, defaultValue=0, description=""}
  User.RibbonIDOffset {type=NiagaraInt32}
  User.SkeletalMesh {
   type=NiagaraDataInterfaceSkeletalMesh
   defaultValue {
    struct=NiagaraExt_VariableValue_DataInterface
    value {dataInterface=&self:NiagaraDataInterfaceSkeletalMesh_1}
   }
  }
 ]
}
